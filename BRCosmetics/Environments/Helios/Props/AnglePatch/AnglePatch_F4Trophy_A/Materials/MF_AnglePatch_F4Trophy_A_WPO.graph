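# Material function for the AnglePatch F4 trophy prop that adds animated vertex bulging and an optional glass orb look to incoming material attributes.
# The Use Glass switch blends in a tinted glass look, lerping normal, specular and roughness by a Fresnel-style term.
# WPO pushes vertices along normals using scrolling Perlin noise, gated by looping wobble and move timers, then clamped and masked.
# Airborne bob and axis-rotation branches exist but never reach the output.
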MaterialFunction /BRCosmetics/Environments/Helios/Props/AnglePatch/AnglePatch_F4Trophy_A/Materials/MF_AnglePatch_F4Trophy_A_WPO
graph {nodes [
  n1 = GetMaterialAttributes(None=FunctionInput(Preview=NamedRerouteUsage()) {inputName=MA, inputType=FunctionInput_MaterialAttributes}) {attributeGetTypes [
    #1
    #2
    #3
    #4
    #5
   ]}
  n2 = NamedRerouteDeclaration(None=VectorParameter() {group=Orb_Anim, parameterName=Orb | Glass Color, defaultValue=<rgba 0.2682 0.4375 0.5 1>}.RGB) {name=Glass Color, nodeColor=<rgba 0.875 1 0 1>}
  n3 = NamedRerouteUsage()
  n4 = Multiply(A=Add(A=@n3) {constB=-0.8}) {constB=20}
  n5 = Add(A=@n3) {constB=-0.6}
  n6 = LinearInterpolate(
   A=Multiply(A=@n1.BaseColor) {constB=0.3}
   B=LinearInterpolate(A=Multiply(A=@n2) {constB=0.3}, B=@n2, Alpha=Saturate(None=@n4))
   Alpha=Saturate(None=Multiply(A=@n5) {constB=20})
  )
  n7 = Saturate(None=Multiply(A=NamedRerouteUsage()) {constB=2})
  n8 = NamedRerouteUsage()
  n9 = Divide(A=Add(A=LocalPosition().Z) {constB=-130}) {constB=40}
  n10 = Multiply(A=Add(A=Saturate(None=@n9)) {constB=0.2}, B=ScalarParameter() {defaultValue=0.1, group=Orb_Anim, parameterName=Orb | Glass Opacity}) {constB=0.1}
  n11 = ScalarParameter() {defaultValue=0.2, desc=0.2, group=Orb_Anim, parameterName=Orb | PBR_Rubber_Specular Min}
  n12 = Saturate(None=LinearInterpolate(
    A=ScalarParameter() {defaultValue=1.5, desc=1.5, group=Orb_Anim, parameterName=Orb | Glass Fresnel_Max}
    B=ScalarParameter() {defaultValue=-0.3, desc=-0.3, group=Orb_Anim, parameterName=Orb | Glass Fresnel_Min}
    Alpha=NamedRerouteUsage()
   ))
  n13 = ScalarParameter() {defaultValue=0.3, desc=0.3, group=Orb_Anim, parameterName=Orb | PBR_Rubber_Roughness Min}
  n14 = VertexNormalWS()
  n15 = SetMaterialAttributes!(
   MaterialAttributes=@n1.MaterialAttributes
   BaseColor=Add(A=LinearInterpolate(A=@n1.BaseColor, B=@n6, Alpha=@n7), B=Multiply(A=@n8, B=@n10) {constB=0.07})
   Normal=LinearInterpolate(A=@n1.Normal, B=Transform(None=@n14) {transformSourceType=TRANSFORMSOURCE_World, transformType=TRANSFORM_Tangent}, Alpha=@n12) {constAlpha=1}
   Specular=LinearInterpolate(
    A=LinearInterpolate(A=@n11, Alpha=@n1.Specular)
    B=ScalarParameter() {defaultValue=0.5, desc=0.5, group=Orb_Anim, parameterName=Orb | PBR_Glass_Specular}
    Alpha=@n12
   ) {constAlpha=1}
   Roughness=LinearInterpolate(
    A=LinearInterpolate(A=@n13, Alpha=@n1.Roughness)
    B=ScalarParameter() {defaultValue=0.1, desc=0.1, group=Orb_Anim, parameterName=Orb | PBR_Glass_Roughness}
    Alpha=@n12
   ) {constAlpha=1}
  ) {attributeSetTypes [
    #1
    #2
    #4
    #5
   ]}
  n16 = BlendMaterialAttributes(A=@n1.MaterialAttributes, B=@n15, Alpha=NamedRerouteUsage())
  n17 = MaterialFunctionCall(Color=LocalPosition().XYZ) {materialFunction=&SplitComponents}
  n18 = AppendVector(A=@n17.B, B=@n17.B)
  n19 = NamedRerouteUsage()
  n20 = Add(A=Sine(None=@n19))
  n21 = Multiply(A=Subtract(A=@n19, B=Multiply(A=@n20) {constB=0.3})) {constB=0.1}
  n22 = Constant2Vector() {g=-0.23, r=-0.47}
  n23 = TextureSample(UVs=Add(A=Divide(A=@n18) {constB=400}, B=Multiply(A=@n21, B=@n22))) {samplerType=SAMPLERTYPE_Masks, texture=/Game/Environments/Helios/Props/AnglePatch/AnglePatch_F4Trophy_A/Textures/T_Generic_Perlin_AnglePatch}
  n24 = Add(A=@n23.RGB) {constB=-0.5}
  n25 = Multiply(
   A=VertexNormalWS()
   B=Multiply(A=Saturate(None=@n24), B=ScalarParameter() {defaultValue=32, group=Orb_Anim, parameterName=Bulge_Mult}) {constB=5}
  )
  n26 = NamedRerouteUsage()
  n27 = ScalarParameter() {defaultValue=3.9, desc=0.56, group=Orb_Anim, parameterName=Anim | Loop Duration}
  n28 = Multiply(A=Frac(None=Divide(A=@n26, B=@n27))) {constB=2}
  n29 = NamedRerouteDeclaration(None=Abs(None=Add(A=@n28) {constB=-1})) {name=Time_V_Slow_1, nodeColor=<rgba 0.6641 1 0 1>}
  n30 = ScalarParameter() {defaultValue=0.85, desc=-0.8, group=Orb_Anim, parameterName=Anim | Not Wobbling % time}
  n31 = Divide(B=OneMinus(None=@n30))
  n32 = Saturate(None=Multiply(A=Subtract(A=@n29, B=@n30), B=@n31))
  n33 = Multiply(A=Time(), B=Multiply(A=@n29, B=ScalarParameter() {defaultValue=5, group=Orb_Anim, parameterName=Wobble Freq})) {constB=4}
  n34 = Add(A=Multiply(A=Frac(None=@n33)) {constB=2}) {constB=-1}
  n35 = Multiply(A=Saturate(None=Abs(None=@n34)), B=ScalarParameter() {defaultValue=0.7, group=Orb_Anim, parameterName=Wobble Intensity})
  n36 = ScalarParameter() {defaultValue=0.6, desc=-0.8, group=Orb_Anim, parameterName=Anim | Not Moving % time}
  n37 = OneMinus(None=@n36)
  n38 = Multiply(A=Subtract(A=@n29, B=@n36), B=Multiply(A=Divide(B=@n37), B=Constant() {desc=Transition Ramp, r=2.5}) {constB=1.5})
  n39 = Add(A=Multiply(A=@n25, B=Multiply(A=@n32, B=@n35)), B=Multiply(A=@n25, B=Saturate(None=@n38)) {constB=5})
  n40 = SetMaterialAttributes!(
   MaterialAttributes=GetMaterialAttributes(None=StaticSwitchParameter(True=@n16, False=@n1.MaterialAttributes) {group=Orb_Anim, parameterName=Use Glass}) {attributeGetTypes [#2]}.MaterialAttributes
   WorldPositionOffset=Multiply(A=Clamp(None=@n39) {maxDefault=15, minDefault=-15}, B=NamedRerouteUsage())
   BaseColor
  ) {attributeSetTypes [#6, #1]}
  n41 = GetMaterialAttributes(None=@n40) {attributeGetTypes [#6]}
  n42 = NamedRerouteUsage()
  n43 = NamedRerouteDeclaration(None=Sine(None=@n42)) {name=Time_Sine, nodeColor=<rgba 0.3125 0 1 1>}
  n44 = LinearInterpolate(A=Sine(None=Multiply(A=@n42) {constB=2.3, desc=2.3}), B=@n43, Alpha=Saturate(None=@n43))
  n45 = ScalarParameter() {defaultValue=-0.8, desc=-0.5, group=Orb_Anim, parameterName=Anim | Airborn/Still Bias}
  n46 = OneMinus(None=Saturate(None=@n45))
  n47 = NamedRerouteUsage()
  n48 = ScalarParameter() {defaultValue=0.3, desc=0.56, group=Orb_Anim, parameterName=Anim | Airborn/Still Freq}
  n49 = Multiply(A=Frac(None=Multiply(A=@n47, B=@n48) {constB=0.5})) {constB=2}
  n50 = NamedRerouteDeclaration(None=Abs(None=Add(A=@n49) {constB=-1})) {name=Time_V_Slow, nodeColor=<rgba 0.6641 1 0 1>}
  n51 = Saturate(None=Add(A=Multiply(A=@n50, B=@n46), B=@n45))
  n52 = NamedRerouteDeclaration(None=Saturate(None=Multiply(A=@n44, B=@n51))) {name=Is_Airborn, nodeColor=<rgba 0.875 0 1 1>}
  n53 = NamedRerouteUsage()
  n54 = Multiply(A=NamedRerouteUsage(), B=Constant3Vector() {constant=<rgba 0 0 0.5 0>})
  n55 = NamedRerouteUsage()
  n56 = NamedRerouteUsage()
  n57 = Multiply(A=Max(A=Multiply(A=@n52, B=@n53), B=Multiply(A=@n54, B=@n55)) {constB=0}, B=@n56)
  n58 = GetMaterialAttributes(None=SetMaterialAttributes!(MaterialAttributes=@n41.MaterialAttributes, WorldPositionOffset=Add(A=@n41.WorldPositionOffset, B=@n57)) {attributeSetTypes [#6]}) {attributeGetTypes [#2]}
  n59 = GetMaterialAttributes(None=@n58.MaterialAttributes) {attributeGetTypes [#6]}
  n60 = Frac(None=Multiply(A=NamedRerouteUsage(), B=ScalarParameter() {defaultValue=2.1, group=Orb_Anim, parameterName=Anim | Rotation Anim Speed}) {constB=0.5})
  n61 = Abs(None=Add(A=Multiply(A=@n60) {constB=2}) {constB=-1})
  n62 = ScalarParameter() {defaultValue=2, group=Orb_Anim, parameterName=Anim | Orb_SideAnim_Mult}
  n63 = Multiply(A=Add(A=NamedRerouteDeclaration(None=@n61) {name=Time_V, nodeColor=<rgba 0.8047 1 0 1>}) {constB=-0.5}, B=@n62) {constB=2}
  n64 = Constant() {r=1}
  n65 = Multiply(
   A=NamedRerouteDeclaration(None=MaterialFunctionCall(R=@n63, G=@n62, B=@n64) {materialFunction=&AppendMany}.RGB) {name=T_RotationAnim, nodeColor=<rgba 0 1 0.9297 1>}
   B=LinearInterpolate(
    A=VectorParameter() {group=Orb_Anim, parameterName=Anim | Rotation Axis_Orb, defaultValue=<rgba 0.1 0.1 1 0>}.RGB
    B=VectorParameter() {group=Orb_Anim, parameterName=Anim | Rotation Axis_Base, defaultValue=<rgba -0.1 -0.1 -10 0>}.RGB
    Alpha=NamedRerouteUsage()
   )
  )
  n66 = NamedRerouteUsage()
  n67 = Multiply(A=@n66) {constB=0.03}
  n68 = NamedRerouteUsage()
  n69 = Multiply(A=@n66) {constB=-0.005}
  n70 = MaterialFunctionCall!(
   RotationAngle=LinearInterpolate(A=Multiply(A=@n67, B=@n68), B=Multiply(A=@n68, B=@n69), Alpha=NamedRerouteUsage())
   NormalizedRotationAxis=Transform(None=Normalize(VectorInput=@n65)) {transformSourceType=TRANSFORMSOURCE_Local}
   PivotPoint=TransformPosition(None=VectorParameter() {group=Orb_Anim, parameterName=Anim | Pivot Position, defaultValue=<rgba 0 0 -130 1>}.RGB) {transformType=TRANSFORMPOSSOURCE_World}
   Position=WorldPosition().XYZ
   Normal=Transform(None=@n58.Normal)
   TangentU
   TangentV
  ) {materialFunction=/BRCosmetics/Characters/Player/Male/Medium/Bodies/M_MED_Inferno/FX/Mat/RotateVertexAboutAxis}
  BlendMaterialAttributes(
   A=@n59.MaterialAttributes
   B=SetMaterialAttributes!(
    MaterialAttributes=@n59.MaterialAttributes
    WorldPositionOffset=Add(A=@n59.WorldPositionOffset, B=@n70.WPO)
    Normal=Transform(None=@n70.Normal) {transformSourceType=TRANSFORMSOURCE_World, transformType=TRANSFORM_Tangent}
   ) {attributeSetTypes [#6, #2]}
   Alpha=NamedRerouteUsage()
  )
  FunctionOutput(None=@n40) {outputName=MA}
  n71 = LocalPosition() {localOrigin=InstancePreSkinning}
  n72 = Multiply(A=@n71.XYZ, B=NamedRerouteUsage())
  n73 = Divide(A=Add(A=@n72, B=Constant3Vector() {constant=<rgba 0 0 -110 0>}) {constB=-110}) {constB=50}
  n74 = MaterialFunctionCall(Color=@n73) {materialFunction=&SplitComponents}
  Max(A=@n74.R, B=@n74.R)
  n75 = Constant3Vector() {constant=<rgba 20 20 10 0>}
  n76 = NamedRerouteUsage()
  Multiply(
   A=@n75
   B=MaterialFunctionCall(R=@n76, G=@n76, B=@n56) {materialFunction=&AppendMany}.RGB
  )
  Multiply(A=@n52, B=@n75)
  Multiply(A=@n31) {constB=2}
  n77 = Add(A=@n72, B=Constant3Vector() {constant=<rgba 0 0 -137 0>}) {constB=-110}
  n78 = OneMinus(None=Abs(None=Divide(A=@n77) {constB=22.5}))
  NamedRerouteDeclaration(None=Saturate(None=ComponentMask(None=@n78) {b=true})) {name=M_OrbGradient_Z, nodeColor=<rgba 0 1 0.0156 1>}
  NamedRerouteDeclaration(None=Multiply(A=@n74.R, B=@n74.R)) {name=M_OrbGradient_XY, nodeColor=<rgba 0 1 0.0156 1>}
  n79 = VertexColor()
  NamedRerouteDeclaration(None=Max(
    A=NamedRerouteDeclaration(None=@n79.R) {name=M_Orb, nodeColor=<rgba 1 0.5391 0 1>}
    B=NamedRerouteDeclaration(None=@n79.G) {name=M_Base, nodeColor=<rgba 0.2188 0 1 1>}
   )) {name=M_Animated, nodeColor=<rgba 0.1016 1 0 1>}
  NamedRerouteDeclaration(None=Multiply(A=Time(), B=ScalarParameter() {defaultValue=1, group=Orb_Anim, parameterName=Anim | Speed, sortPriority=0}) {constB=0.5}) {name=In_Time, nodeColor=<rgba 0.6172 0 1 1>}
  n80 = TextureSample() {samplerType=SAMPLERTYPE_Masks, texture=/Game/Environments/Helios/Props/AnglePatch/AnglePatch_F4Trophy_A/Textures/T_AnglePatch_F4Trophy_A_A_S}
  n81 = TextureSample() {
   samplerType=SAMPLERTYPE_Normal
   texture=/Game/Environments/Helios/Props/AnglePatch/AnglePatch_F4Trophy_A/Textures/T_AnglePatch_F4Trophy_A_A_N
  }
  NamedRerouteDeclaration(None=MakeMaterialAttributes!(
    BaseColor=TextureSample() {texture=/Game/Environments/Helios/Props/AnglePatch/AnglePatch_F4Trophy_A/Textures/T_AnglePatch_F4Trophy_A_A_D}.RGB
    Metallic=@n80.G
    Specular=@n80.G
    Roughness=@n80.G
    Anisotropy
    EmissiveColor
    Opacity
    OpacityMask
    Normal=@n81.RGB
    Tangent
    WorldPositionOffset
    SubsurfaceColor
    ClearCoat
    ClearCoatRoughness
    AmbientOcclusion
    Refraction
    CustomizedUVs_0
    CustomizedUVs_1
    CustomizedUVs_2
    CustomizedUVs_3
    CustomizedUVs_4
    CustomizedUVs_5
    CustomizedUVs_6
    CustomizedUVs_7
    PixelDepthOffset
    ShadingModel
    Displacement
   )) {name=MA_Preview, nodeColor=<rgba 0 0.0625 1 1>}
  n82 = DotProduct(A=CameraVectorWS(), B=VertexNormalWS())
  n83 = Multiply(A=Add(A=NamedRerouteDeclaration(None=@n82) {name=Fresnel_Pure, nodeColor=<rgba 1 0.8672 0 1>}) {constB=-0.2}) {constB=3}
  NamedRerouteDeclaration(None=OneMinus(None=Saturate(None=@n83))) {name=Fresnel_GlassThickness, nodeColor=<rgba 0 0.6953 1 1>}
  Saturate(None=@n73)
  TextureSample()
  VertexInterpolator(VS=@n71.XYZ)
 ]}
